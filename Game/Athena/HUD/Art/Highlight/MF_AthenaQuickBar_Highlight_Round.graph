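# Material function for the Athena quick bar highlight art that outputs a round, linear-falloff gradient mask.
# It offsets UVs by the Center X/Y inputs, scales by Circle Size, then clamps and inverts the distance from zero.
# The mask is brightest at the center, and a larger Circle Size gives a smaller circle.
# Default inputs center the circle and reach zero at the UV edge midpoints.
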MaterialFunction /Game/Athena/HUD/Art/Highlight/MF_AthenaQuickBar_Highlight_Round
graph {nodes [
  n1 = TextureCoordinate()
  n2 = AppendVector(
   A=FunctionInput(Preview=Constant() {r=0.5}) {bUsePreviewValueAsDefault=true, inputName=Center X, inputType=FunctionInput_Scalar, sortPriority=1}
   B=FunctionInput(Preview=Constant() {r=0.5}) {bUsePreviewValueAsDefault=true, inputName=Center Y, inputType=FunctionInput_Scalar, sortPriority=2}
  )
  n3 = Constant() {r=2}
  n4 = Distance(
   A=Multiply(
    A=Subtract(A=@n1, B=@n2)
    B=FunctionInput(Preview=@n3) {bUsePreviewValueAsDefault=true, inputName=Circle Size, inputType=FunctionInput_Scalar, sortPriority=3}
   )
   B=Constant()
  )
  FunctionOutput(None=OneMinus(None=Clamp(None=@n4)))
 ]}
