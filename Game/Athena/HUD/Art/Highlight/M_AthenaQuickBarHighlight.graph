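# Additive UI material that draws a rounded-rectangle highlight, named for the Athena HUD quick bar.
# It subtracts an inset GeneratedRoundRect from an outer one to form a border with parameterized size, corners and sharpness.
# A static switch optionally adds a powered, inverted radial gradient masked to the outer shape, merged by Lighten blend.
# The Color parameter's alpha scales the clamped result; Center X and Y parameters are wired in swapped order.
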Material /Game/Athena/HUD/Art/Highlight/M_AthenaQuickBarHighlight
graph {
 properties {blendMode=BLEND_Additive, materialDomain=MD_UI, pixelDepthOffsetMode=PDOM_Legacy, usedShadingModels=Default Lit}
 nodes [
  n1 = TextureCoordinate()
  n2 = ScalarParameter() {defaultValue=1, group=Square, parameterName=Box Width, sliderMax=1, sortPriority=1}
  n3 = ScalarParameter() {defaultValue=1, group=Square, parameterName=Box Height, sliderMax=1, sortPriority=2}
  n4 = ScalarParameter() {group=Square, parameterName=Corner Radius, sliderMax=1, sortPriority=3}
  n5 = ScalarParameter() {defaultValue=20, group=Square, parameterName=Sharpness, sliderMax=35, sortPriority=4}
  n6 = ScalarParameter() {defaultValue=0.5, group=Square, parameterName=Center Y, sortPriority=5}
  n7 = ScalarParameter() {defaultValue=0.5, group=Square, parameterName=Center X, sortPriority=6}
  n8 = AppendVector(A=@n6, B=@n7)
  n9 = MaterialFunctionCall!(UV Coords=@n1, Box Dimensions=AppendVector(A=@n2, B=@n3), CornerRadius=@n4, Sharpness=@n5, Center=@n8) {materialFunction=/Engine/Functions/Engine_MaterialFunctions01/Gradient/GeneratedRoundRect}
  n10 = ScalarParameter() {defaultValue=0.1, group=Square, parameterName=Box Border Width, sortPriority=7}
  n11 = Subtract(A=@n2, B=@n10)
  n12 = Subtract(A=@n3, B=@n10)
  n13 = Subtract(
   A=@n9.Result
   B=MaterialFunctionCall(UV Coords=@n1, Box Dimensions=AppendVector(A=@n11, B=@n12), CornerRadius=@n4, Sharpness=@n5, Center=@n8) {materialFunction=&GeneratedRoundRect}.Result
  )
  n14 = ScalarParameter() {defaultValue=1, group=Gradient, parameterName=Gradient Size, sortPriority=1}
  n15 = Power(
   Base=OneMinus(None=MaterialFunctionCall!(Center X=@n6, Center Y=@n7, Circle Size=@n14) {materialFunction=&/MF_AthenaQuickBar_Highlight_Round}.Result)
   Exp=ScalarParameter() {defaultValue=1, group=Gradient, parameterName=Gradient Power, sortPriority=2}
  )
  n16 = StaticSwitchParameter(True=Multiply(A=@n9.Result, B=Saturate(None=@n15)), False=Constant()) {group=Gradient, parameterName=Gradient}
  Clamp(None=Multiply(
    A=VectorParameter() {group=Square, parameterName=Color, sortPriority=8, defaultValue=<rgba 1 1 1 1>}.A
    B=MaterialFunctionCall!(Blend=@n13, Base=@n16) {materialFunction=/Engine/Functions/Engine_MaterialFunctions03/Blends/Blend_Lighten}.Result
   ))
 ]
}
